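# Void Slide HUD widget that shows input hints for the Change Seat and Exit actions.
# Construct, keybind changes and picture-in-picture (PiP) block changes queue a flag-guarded, next-tick refresh of the feedback lines.
# AddFeedbackLine fills an existing line by index from each action definition, adding secondary input actions only on gamepad.
# ShowFeedback and HideFeedback bind and unbind the PiP blocking event, which also switches the text style.
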WidgetBlueprint /WildEstate/Gameplay/VoidSlide/UI/WBP_VoidSlideFeedback
graphs [
 AddFeedbackLine {nodes [
   n1 = AddFeedbackLine!() -> (then, ActionDef:Void Slide Action Def Structure, ActionIndex:int) {is=title K2Node_FunctionEntry}
   > Branch(Condition="integer < integer"!(A:int=@n1.ActionIndex, B:int=Get MaxLinesNeeded() -> (MaxLinesNeeded:int=0)) -> (ReturnValue:bool) {is=title K2Node_CallFunction})
   > Set ActionWidget(ActionWidget:Athena HUDFeedback Line Object Reference="Get (a copy)"(
     Array:Array of WBP Athena HUDFeedback Line New CD Object References=Get FeedbackLines() -> (FeedbackLines:Array of WBP Athena HUDFeedback Line New CD Object References)
     Dimension 1=Get ActionIndex() -> (ActionIndex:int=0)
    ) -> (Output:WBP Athena HUDFeedback Line New CD Object Reference)) -> (Output_Get:Athena HUDFeedback Line Object Reference)
   > n2 = For Each Loop(Array:Array of Input Action Object References=Get!() -> (
     ActionDef_ActionLabel_17_#1:Text
     ActionDef_InputActions_16_#2:Array of Input Action Object References
     ActionDef_DisabledTags_18_#3:Gameplay Tag Container Structure
    ).ActionDef_InputActions_16_#2) -> (Array Element:Input Action Object Reference)
   > Set InputAction(InputAction:Input Action Object Reference=@n2) -> (Output_Get:Input Action Object Reference)
   > n3 = Branch(Condition="Equal (Integer)"(A=@n2.Array Index))
   n4 = Get InputAction() -> (InputAction:Input Action Object Reference)
   > Is Valid(InputObject=@n4)
   > SetFeedbackFromInputAction!(
    execute
    self:Athena HUDFeedback Line Object Reference=Get ActionWidget() -> (ActionWidget:Athena HUDFeedback Line Object Reference)
    NewBoundAction:Input Action Object Reference=@n4
    FeedbackText:"Text (by ref)"=Get()
    bShowEnabled:bool=true
    AxisScale:float=0
    bHasMultipleFeedbackLines:bool=false
   ) -> (then)
   > SetVisibility(self=Get ActionWidget(), InVisibility=SelfHitTestInvisible)
   > SetDisabledTags!(
    execute
    self:Athena HUDFeedback Line Base Object Reference=Get ActionWidget()
    InDisabledTags:"Gameplay Tag Container Structure (by ref)"=Get().ActionDef_DisabledTags_18_#3
   ) -> (then)
   > SetDividerVisible!(execute, self:Athena HUDFeedback Line Object Reference=Get ActionWidget(), bIsVisible:bool=false) -> (then)
   > Branch(Condition=Get bShouldClearFeedbackPanel() -> (bShouldClearFeedbackPanel:bool))
   > AddChild(self=Get Feedback() -> (Feedback:Vertical Box Object Reference), Content=Get ActionWidget())
   n5 = GetOwningPlayer()
   n6 = "Not Equal (Enum)"!(
    A:ECommonInputType Enum=GetCurrentInputType(self=Get CommonInputSubsystem!(PlayerController:Player Controller Object Reference=@n5) -> (ReturnValue:Common Input Object Reference) {is=title K2Node_GetSubsystemFromPC})
    B:ECommonInputType Enum=Gamepad
   ) -> (ReturnValue:bool)
   Branch(execute=@n3.else, Condition=NOT Boolean(A=@n6))
   > AddBoundAction!(
    execute
    self:Athena HUDFeedback Line Object Reference=Get ActionWidget()
    NewBoundLegacyAction:Name=None
    NewBoundEnhancedAction:Input Action Object Reference=Get InputAction()
    AxisScale:float=0
    bShowSecondaryBoundKey:bool=@n6
   ) -> (then)
  ]}
 AttemptRefreshFeedbackLines {nodes [
   AttemptRefreshFeedbackLines() {is=title K2Node_FunctionEntry}
   > n1 = Branch(Condition=Get bFeedbackRefreshQueued() -> (bFeedbackRefreshQueued:bool))
   Set Timer for Next Tick by Event!(execute=@n1.else, Delegate:Delegate=Create Event()) -> (then, ReturnValue:Timer Handle Structure) {is=title K2Node_CallFunction}
   > Set bFeedbackRefreshQueued(bFeedbackRefreshQueued:bool=true) -> (Output_Get:bool)
  ]}
 EventGraph {nodes [
   n1 = OnPiPBlockPlayer!() -> (OutputDelegate, then, PIPIsBlocking:bool=~) {is=title K2Node_CustomEvent}
   > AttemptRefreshFeedbackLines() {is=title K2Node_CallFunction}
   > n2 = Branch(Condition=@n1.PIPIsBlocking)
   > UpdateFeedbackStyles!(execute, self:WBP Void Slide Feedback Parent Object Reference, Style:Data Table Object Reference) -> (then)
   UpdateFeedbackStyles(execute=@n2.else, Style=Get DefaultTextStyleSet() -> (DefaultTextStyleSet:Data Table Object Reference))
   OnKeybindsChanged_Bind() {is=title K2Node_CustomEvent}
   > AttemptRefreshFeedbackLines()
   > ShowFeedback(self:WBP Void Slide Feedback Parent Object Reference)
   Event ShowFeedback() {is=title K2Node_Event}
   n3 = GetPiPBlockManager!() -> (ReturnValue:Live Stream Pi PBlock Manager Object Reference)
   > Is Valid(InputObject=@n3)
   > Bind Event to PIPPlayer Blocking Event(self:Live Stream Pi PBlock Manager Object Reference=@n3, Delegate=@n1) {is=title K2Node_AddDelegate}
   > OnPiPBlockPlayer!(execute, self, PIPIsBlocking:bool=IsPlayerBlocked!(self:Live Stream Pi PBlock Manager Object Reference=@n3) -> (ReturnValue:bool)) -> (then) {is=title K2Node_CallFunction}
   > Parent: ShowFeedback() {is=title K2Node_CallParentFunction}
   Event HideFeedback() {is=title K2Node_Event}
   n4 = GetPiPBlockManager()
   > Is Valid(InputObject=@n4)
   > Unbind Event from PIPPlayer Blocking Event(self:Live Stream Pi PBlock Manager Object Reference=@n4, Delegate=Create Event()) {is=title K2Node_RemoveDelegate}
   > UpdateFeedbackStyles(Style=Get DefaultTextStyleSet())
   > Parent: HideFeedback() {is=title K2Node_CallParentFunction}
   Event Construct()
   > Parent: Construct()
   > Set bFeedbackRefreshQueued(bFeedbackRefreshQueued=false)
   > AttemptRefreshFeedbackLines()
   > Bind Event to On Keybinds Changed(
    self:Fort Global UIContext Object Reference=Get FortGlobalUIContext!() -> (ReturnValue:Fort Global UIContext Object Reference) {is=title K2Node_GetBlueprintContextEx}
    Delegate=Create Event()
   ) {is=title K2Node_AddDelegate}
   > n5 = Branch(Condition=Get bShouldClearFeedbackPanel())
   n6 = For Each Loop(Exec=@n5.else, Array:Array of WBP Athena HUDFeedback Line New CD Object References=Get FeedbackLines()) -> (Array Element:WBP Athena HUDFeedback Line New CD Object Reference)
   > SetVisibility(self=@n6, InVisibility=Collapsed)
   > AddChild(self=Get Feedback(), Content=@n6)
   Any Key!() -> (Pressed:Exec, Released:Exec, Key:Key Structure) {is=title K2Node_InputKey, inputKey=AnyKey}
   > ShowFeedback()
  ]}
 RefreshFeedbackLines {nodes [
   RefreshFeedbackLines() {is=title K2Node_FunctionEntry}
   > n1 = For Each Loop(Array=Get FeedbackLines())
   > SetVisibility(self=@n1, InVisibility=Collapsed)
   Set bFeedbackRefreshQueued(execute=@n1.Completed, bFeedbackRefreshQueued=false)
   > n2 = Branch(Condition=Get bShouldClearFeedbackPanel())
   > n3 = ClearChildren(self=Get Feedback())
   n4 = For Each Loop(
    Exec=[@n3.then, @n2.else]
    Array:Array of Void Slide Action Def Structures=Get ActionDefs() -> (ActionDefs:Array of Void Slide Action Def Structures)
   ) -> (Array Element:Void Slide Action Def Structure)
   > AddFeedbackLine!(execute, self, ActionDef:Void Slide Action Def Structure=@n4, ActionIndex:int=@n4.Array Index) -> (then) {is=title K2Node_CallFunction}
  ]}
]
variables [
 ActionDefs {
  type="array<VoidSlideActionDef>"
  default [
   {actionLabel=Change Seat, disabledTags {gameplayTags []}, inputActions [/WildEstate/Gameplay/VoidSlide/Blueprints/VoidSlideChangeSeatInput]}
   {actionLabel=Exit, disabledTags {gameplayTags []}, inputActions [/WildEstate/Gameplay/VoidSlide/Blueprints/VoidSlideExitInput]}
  ]
 }
 bFeedbackRefreshQueued {type=boolean}
 bShouldClearFeedbackPanel {type=boolean}
]
widgets {
 info {
  inheritedWidgetCount=0
  parentClass=&/WBP_VoidSlideFeedback_Parent.WBP_VoidSlideFeedback_Parent_C
  rootWidgetClass=null
  widgetCount=0
 }
 widgets []
}
